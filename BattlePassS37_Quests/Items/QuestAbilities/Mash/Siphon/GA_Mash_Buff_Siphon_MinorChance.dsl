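; Gameplay ability GA_Mash_Buff_Siphon_MinorChance is a chance-gated variant of a siphon buff ability for the Mash quest items.
; When activated from a gameplay event, it rolls a weighted random bool using the ChanceToSiphon scalable float.
; Only on a successful roll does it call the parent activation with the event data, so the siphon effect itself lives in the parent class.
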(event Ability|EventActivateAbilityFromEvent
  (params (EventData :type Gameplay%20Event%20Data%20Structure%20%28by%20ref%29))
  (if (Math|Random|RandomBoolwithWeight (ScalableFloat|GetValueatLevel (Variables|Default|GetHF_ChanceToSiphon)))
    (|CallParentFunction EventData)))
